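# Sound cue that plays the Hightower JumpFire sound from the WarmWest audio set.
# A single wave player feeds the output directly, with a volume boost and no randomization or mixing.
# The cue routes through the Fort_Traps_Placement sound class.
# It overrides attenuation inline for long-range audibility, with a wide stereo spread and a non-spatialized radius near the source.
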SoundCue /HighTower/Audio/WarmWest/Hightower_JumpFire_Cue
root=WavePlayer_0
nodes [{id=WavePlayer_0, properties {soundWaveAssetPtr=&/Hightower_JumpFire}}]
properties {
 bOverrideAttenuation=true
 maxDistance=14000
 soundClassObject=/Game/Sounds/FortSoundClassesAndModes/Fort_Traps_Placement
 volumeMultiplier=1.35
 attenuationOverrides {falloffDistance=10000, nonSpatializedRadiusStart=500, stereoSpread=5000, attenuationShapeExtents {x=4000}}
}
