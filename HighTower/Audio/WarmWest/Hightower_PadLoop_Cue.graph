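# Hightower_PadLoop_Cue is a Sound Cue that plays a looping pad sound for HighTower's WarmWest audio set.
# A single looping Wave Player feeds the cue root directly, with no randomization or mixing nodes.
# The cue plays at reduced volume in the Fort_Gadgets sound class and overrides attenuation inline with a natural-sound falloff, so it fades with distance from the source.
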SoundCue /HighTower/Audio/WarmWest/Hightower_PadLoop_Cue
root=WavePlayer_0
nodes [{id=WavePlayer_0, properties {bLooping=true, soundWaveAssetPtr=&/Hightower_PadLoop}}]
properties {
 bOverrideAttenuation=true
 maxDistance=3826.146
 soundClassObject=&Fort_Gadgets
 volumeMultiplier=0.3
 attenuationOverrides {distanceAlgorithm=NaturalSound, attenuationShapeExtents {x=226.1459}}
}
